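# Conversation database for the bounty board in the player-hunting bounty feature, opened by the BountyBoard entry tag with an NPC participant as a speaker.
# The entry point leads to a Speech node, which leads to a Continue node offering a single Decline choice tagged as Back with a back icon.
# The graph contains no accept branch or bounty-specific task nodes.
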ConversationDatabase /Bounties/HuntingPlayer/Conversations/Conversation_BountyBoard
entry_points [{tag=Conversation.Entry.BountyBoard, destinations [#1]}]
nodes [
 ConversationEntryPointNode_3 {
  id=#1
  class=ConversationEntryPointNode
  outputs [#2]
  properties {entryTag {tagName=Conversation.Entry.BountyBoard}}
 }
 Speech_C_1 {id=#2, class=Speech_C, outputs [#3], properties {nodeName=""}}
 Continue_C_0 {
  id=#3
  class=Continue_C
  properties {softTaskIcon=&T_UI_NPC_Back}
  sub_nodes [ConversationChoiceNode_0 {
    class=ConversationChoiceNode
    properties {defaultChoiceDisplayText=Decline, choiceTags {gameplayTags [{tagName=Conversation.Choice.Category.Back}]}}
   }]
 }
]
speakers [{}, {participant=Conversation.Participant.NPC}]
