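# Conversation database for the HuntingPlayer bounty dialogue with an NPC participant, entered through the Conversation.Entry.HuntingPlayer tag.
# The entry point leads to a linear chain of two Speech nodes, each with one choice sub-node: "How can I help?", then "Decline".
# The chain ends at Decline; no accept branch, spoken lines or bounty logic are set here.
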ConversationDatabase /Bounties/HuntingPlayer/Conversations/Conversation_HuntingPlayer
entry_points [{tag=Conversation.Entry.HuntingPlayer, destinations [#1]}]
nodes [
 ConversationEntryPointNode_7 {
  id=#1
  class=ConversationEntryPointNode
  outputs [#2]
  properties {entryTag {tagName=Conversation.Entry.HuntingPlayer}}
 }
 Speech_C_0 {
  id=#2
  class=Speech_C
  outputs [#3]
  sub_nodes [ConversationChoiceNode_0 {class=ConversationChoiceNode, properties {defaultChoiceDisplayText=How can I help?}}]
 }
 Speech_C_1 {
  id=#3
  class=Speech_C
  outputs []
  sub_nodes [ConversationChoiceNode_2 {class=ConversationChoiceNode, properties {defaultChoiceDisplayText=Decline}}]
 }
]
speakers [{}, {participant=Conversation.Participant.NPC}]
